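# Niagara feather burst for the Nug prey AI, spawning mesh feathers on its skeletal mesh.
# The active zombiefeathers emitter bursts more, faster feathers when bIsRunning, and none when TurnOffFeathersForChrome is set.
# Tint samples User.ColorCurve at ColorCurveDataValue snapped to tenth-wide bin centers, with random brightness variation.
# Feathers fly outward under drag, gravity and curl noise, slowing their spin, fading and shrinking; a similar recolor_feathers emitter is disabled.
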NiagaraSystem /Irwin/AI/Prey/Nug/FX/NS_Irwin_Nug_FeatherBurst
dependencies {usedDataInterfaces [&NiagaraDataInterfaceCurve], usedDynamicInputs [], usedModules [&SystemState]}
emitters {
 recolor_feathers {
  bEnabled=false
  emitterUpdateScript [
   &EmitterState(
    Enable Distance Culling=true
    Inactive Response=::NewEnumerator1
    Life Cycle Mode=::NewEnumerator1
    Loop Behavior=::NewEnumerator1
    Loop Duration=2.1
    Max Distance Response=::KillAfterParticlesFinish
    Module.bUseMaxDistance=true
    Module.bUseMinDistance=true
    Scalability Mode=::NewEnumerator1
    UseLoopDelay=false
   )
   &SpawnRate(SpawnRate=150, Use Spawn Probability=true) {enabled=false}
   &SpawnBurst_Instantaneous(
    Spawn Count=&MakeCustomIntFromBool(Bool=$User.bIsRunning, True INT=15, False INT=5)
    Spawn Probability=&MakeCustomFloatFromBool(Bool=$User.TurnOffFeathersForChrome, True Float=0, False Float=1)
    Spawn Time=0.01
    Use Spawn Probability=true
    UseLoopCountLimit=false
   )
  ]
  particleSpawnScript [
   /Niagara/Modules/Spawn/Initialization/InitializeParticle(
    Color=&MultiplyLinearColor(
     Linear Color=&MultiplyLinearColor(
      Linear Color=&ColorFromCurve(
       ColorCurve=$User.ColorCurve
       CurveIndex=&Divide_Float(
        A=&Add_Float(
         A=/Niagara/DynamicInputs/Math/Floor_Float(Float:NiagaraFloat=&Multiply_Float(A=$User.ColorCurveDataValue, B=10))
         B=0.5
        )
        B=10
       )
      )
      Scale Factor=&MakeFloat4FromVectorAndFloat(
       "Vector (RGB)"=&VectorFromFloat(Value=&FloatFromCurve(
         FloatCurve=DataInterface() {propertyValues {Curve {keys [{arriveTangent=0.7728, interpMode=RCIM_Cubic, leaveTangent=0.7728, tangentMode=RCTM_User, time=0, value=0.5}, <at 0.5 1>, <at 1 1>]}}}
         CurveIndex=&RandomRangeFloat(Minimum=0, Maximum=1)
        ))
       "Float (Alpha)"=1
      )
     )
     Scale Factor=<1 1 1 1>
    )
    Lifetime=&FloatFromCurve(
     FloatCurve=DataInterface() {propertyValues {Curve {keys [
         {arriveTangent=0.0237, interpMode=RCIM_Cubic, leaveTangent=0.0237, tangentMode=RCTM_User, time=0, value=0.6}
         {arriveTangent=0.0589, interpMode=RCIM_Cubic, leaveTangent=0.0589, tangentMode=RCTM_User, time=0.6, value=0.7}
         {arriveTangent=0.1241, interpMode=RCIM_Cubic, leaveTangent=0.1241, tangentMode=RCTM_User, time=1, value=1}
        ]}}}
     CurveIndex=&RandomRangeFloat(Minimum=0, Maximum=1)
     Scale Curve=2
    )
    Mesh Scale=&Multiply_VectorByFloat(Vector=&RandomRangeVector(Minimum=<0.6 0.6 1.5>, Maximum=<1 1 2.5>), Float=0.75)
    Module.Write Color=true
    Module.Write Lifetime=true
    Module.Write Mass=true
    Module.Write Material Random=true
    Module.Write Position=false
    Module.Write Scale=true
    Module.Write SpriteRotation=true
    Module.Write SpriteSize=true
    Position=&SimulationPosition()
    Sprite Rotation=&RandomRangeFloat(Minimum=0, Maximum=360)
   )
   &SkeletalMeshLocation(Inherit Velocity=false, MaskSpawnGroup=false, Module.Override Randomness=false, Skeletal Mesh=$User.SkeletalMesh)
   /Niagara/Modules/Update/Mass/CalculateSizeAndRotationalInertiaByMass(Density by Material Type:ENiagaraMassByVolume=::NewEnumerator4, Depth=0, Height=5, Module.Manually Enter Density=false)
   &InitialMeshOrientation(
    Orientation Vector=<1 1 1>
    Rotation=&RandomRangeVector(Minimum=<0 0 0>, Maximum=<1 1 1>)
    Use Orientation Vector=true
    Use Rotation Vector=true
   )
   &ApplyInitialForces()
   /SetVariables(Particles.RandRotForce:NiagaraFloat=&RandomRangeFloat(Minimum=-6, Maximum=6)) {isSetParameters=true, module=SetVariables_#1}
   &AddVelocityFromPoint(
    Module.Use Falloff Distance=false
    Velocity Strength=&FloatFromCurve(
     FloatCurve=DataInterface() {propertyValues {Curve {keys [
         {interpMode=RCIM_Cubic, tangentMode=RCTM_User, time=0, value=0.05}
         {arriveTangent=0.95, interpMode=RCIM_Cubic, leaveTangent=0.95, tangentMode=RCTM_User, time=0.8, value=0.3}
         {arriveTangent=0.6314, interpMode=RCIM_Cubic, leaveTangent=0.6314, tangentMode=RCTM_User, time=1, value=1}
        ]}}}
     CurveIndex=&RandomRangeFloat(Minimum=0, Maximum=1)
     Scale Curve=&MakeCustomFloatFromBool(Bool=$User.bIsRunning, True Float=1300, False Float=800)
    )
   )
   &InheritVelocity(Module.Limit Speed=true, Velocity Scale=<0.1 0.1 0.1>)
  ]
  particleUpdateScript [
   &ScaleColor(
    Scale Alpha=&FloatFromCurve(FloatCurve=DataInterface() {propertyValues {Curve {keys [
         <at 0 1>
         {arriveTangent=-0.0304, interpMode=RCIM_Cubic, leaveTangent=-0.0304, tangentMode=RCTM_User, time=0.6, value=1}
         {arriveTangent=-2.4154, interpMode=RCIM_Cubic, leaveTangent=-2.4154, tangentMode=RCTM_User, time=1, value=0}
        ]}}})
    ScaleA=true
    ScaleRGB=true
   )
   &Drag(Drag=6, Use Linear Drag=true, Use Rotational Drag=false)
   &AccelerationForce(Acceleration=&RandomRangeVector(Minimum=<0 0 -10>, Maximum=<0 0 75>)) {enabled=false}
   &GravityForce(Gravity=&RandomRangeVector(Minimum=<0 0 -75>, Maximum=<0 0 -200>))
   &CurlNoiseForce(
    Noise Frequency=20
    Noise Strength=&FloatFromCurve(
     FloatCurve=DataInterface() {propertyValues {Curve {keys [
         {interpMode=RCIM_Cubic, tangentMode=RCTM_User, time=0, value=1}
         {arriveTangent=-0.0474, interpMode=RCIM_Cubic, leaveTangent=-0.0474, tangentMode=RCTM_User, time=1, value=0.1}
        ]}}}
     Scale Curve=1200
    )
    Pan Noise=false
   )
   &SolveForcesAndVelocity(Clamp Velocity=false, Limit Acceleration=false, Manually Enable Drag Local Offset=true, Manually Enable Rotational Solver=true)
   &ParticleState()
   &UpdateMeshOrientation(Rotation Vector=&Lerp_Vector(
     A=&Multiply_VectorByFloat(Vector=<0.1 0.1 1>, Float=$Particles.RandRotForce)
     B=<0 0 0>
     Alpha=&FloatFromCurve(FloatCurve=DataInterface() {propertyValues {Curve {keys [
          {arriveTangent=1.5391, interpMode=RCIM_Cubic, leaveTangent=1.5391, tangentMode=RCTM_User, time=0, value=0}
          {arriveTangent=1.9127, interpMode=RCIM_Cubic, leaveTangent=1.9127, tangentMode=RCTM_User, time=1, value=1}
         ]}}})
    ))
   &ScaleMeshSize(Scale Factor=&Multiply_VectorByFloat(
     Vector=<1 1 1>
     Float=&FloatFromCurve(FloatCurve=DataInterface() {propertyValues {Curve {keys [
          <at 0 1>
          {arriveTangent=-0.0008, interpMode=RCIM_Cubic, leaveTangent=-0.0008, tangentMode=RCTM_User, time=0.8, value=1}
          {arriveTangent=-3.9673, interpMode=RCIM_Cubic, leaveTangent=-3.9673, tangentMode=RCTM_User, time=1, value=0.2}
         ]}}})
    ))
  ]
  rendererClasses [&NiagaraMeshRendererProperties]
  renderers [{rendererClass=&NiagaraMeshRendererProperties, rendererIndex=0}]
 }
 zombiefeathers {
  emitterUpdateScript [
   &EmitterState(Inactive Response=::NewEnumerator1, Life Cycle Mode=::NewEnumerator1, Loop Behavior=::NewEnumerator1, Loop Duration=2.1, UseLoopDelay=false)
   &SpawnRate(SpawnRate=150, Use Spawn Probability=true) {enabled=false}
   &SpawnBurst_Instantaneous(
    Spawn Count=&MakeCustomIntFromBool(Bool=$User.bIsRunning, True INT=15, False INT=5)
    Spawn Probability=&MakeCustomFloatFromBool(Bool=$User.TurnOffFeathersForChrome, True Float=0, False Float=1)
    Spawn Time=0.01
    Use Spawn Probability=true
    UseLoopCountLimit=false
   )
  ]
  particleSpawnScript [
   &InitializeParticle(
    Color=&MultiplyLinearColor(
     Linear Color=&MultiplyLinearColor(
      Linear Color=&ColorFromCurve(
       ColorCurve=$User.ColorCurve
       CurveIndex=&Divide_Float(
        A=&Add_Float(
         A=&Floor_Float(Float:NiagaraFloat=&Multiply_Float(A=$User.ColorCurveDataValue, B=10))
         B=0.5
        )
        B=10
       )
      )
      Scale Factor=&MakeFloat4FromVectorAndFloat(
       "Vector (RGB)"=&VectorFromFloat(Value=&FloatFromCurve(
         FloatCurve=DataInterface() {propertyValues {Curve {keys [{arriveTangent=0.7728, interpMode=RCIM_Cubic, leaveTangent=0.7728, tangentMode=RCTM_User, time=0, value=0.5}, <at 0.5 1>, <at 1 1>]}}}
         CurveIndex=&RandomRangeFloat(Minimum=0, Maximum=1)
        ))
       "Float (Alpha)"=1
      )
     )
     Scale Factor=&MakeFloat4FromVectorAndFloat(
      "Vector (RGB)"=&VectorFromFloat(Value=&FloatFromCurve(
        FloatCurve=DataInterface() {propertyValues {Curve {keys [{arriveTangent=0.7728, interpMode=RCIM_Cubic, leaveTangent=0.7728, tangentMode=RCTM_User, time=0, value=0.2}, <at 0.5 1>, <at 1 1>]}}}
        CurveIndex=&RandomRangeFloat(Minimum=0, Maximum=1)
        Scale Curve=&FloatFromCurve(
         FloatCurve=DataInterface() {propertyValues {Curve {keys [<at 0 1 RCIM_Constant>,<at 0.2 1.5 RCIM_Constant>,<at 0.4 2 RCIM_Constant>,<at 0.6 1.5 RCIM_Constant>,<at 0.8 1>,<at 1 1>]}}}
         CurveIndex=$User.ColorCurveDataValue
        )
       ))
      "Float (Alpha)"=&FloatFromCurve(
       FloatCurve=DataInterface() {propertyValues {Curve {keys [{arriveTangent=2.9704, time=0, value=0.8}, <at 0.4 1>, <at 1 1>]}}}
       CurveIndex=&RandomRangeFloat(Minimum=0, Maximum=1)
      )
     )
    )
    Lifetime=&FloatFromCurve(
     FloatCurve=DataInterface() {propertyValues {Curve {keys [
         {arriveTangent=0.0237, interpMode=RCIM_Cubic, leaveTangent=0.0237, tangentMode=RCTM_User, time=0, value=0.6}
         {arriveTangent=0.0589, interpMode=RCIM_Cubic, leaveTangent=0.0589, tangentMode=RCTM_User, time=0.6, value=0.7}
         {arriveTangent=0.1241, interpMode=RCIM_Cubic, leaveTangent=0.1241, tangentMode=RCTM_User, time=1, value=1}
        ]}}}
     CurveIndex=&RandomRangeFloat(Minimum=0, Maximum=1)
     Scale Curve=2
    )
    Mesh Scale=&Multiply_VectorByFloat(Vector=&RandomRangeVector(Minimum=<0.6 0.6 1.5>, Maximum=<1 1 2.5>), Float=0.75)
    Module.Write Color=true
    Module.Write Lifetime=true
    Module.Write Mass=true
    Module.Write Material Random=true
    Module.Write Position=false
    Module.Write Scale=true
    Module.Write SpriteRotation=true
    Module.Write SpriteSize=true
    Position=&SimulationPosition()
    Sprite Rotation=&RandomRangeFloat(Minimum=0, Maximum=360)
   )
   &SkeletalMeshLocation(Inherit Velocity=false, MaskSpawnGroup=false, Module.Override Randomness=false, Skeletal Mesh=$User.SkeletalMesh)
   &CalculateSizeAndRotationalInertiaByMass(Density by Material Type:ENiagaraMassByVolume=::NewEnumerator4, Depth=0, Height=5, Module.Manually Enter Density=false)
   &InitialMeshOrientation(
    Orientation Vector=<1 1 1>
    Rotation=&RandomRangeVector(Minimum=<0 0 0>, Maximum=<1 1 1>)
    Use Orientation Vector=true
    Use Rotation Vector=true
   )
   &ApplyInitialForces()
   /SetVariables(Particles.RandRotForce:NiagaraFloat=&RandomRangeFloat(Minimum=-6, Maximum=6)) {isSetParameters=true, module=SetVariables_#1}
   &AddVelocityFromPoint(
    Module.Use Falloff Distance=false
    Velocity Strength=&FloatFromCurve(
     FloatCurve=DataInterface() {propertyValues {Curve {keys [
         {interpMode=RCIM_Cubic, tangentMode=RCTM_User, time=0, value=0.05}
         {arriveTangent=0.95, interpMode=RCIM_Cubic, leaveTangent=0.95, tangentMode=RCTM_User, time=0.8, value=0.3}
         {arriveTangent=0.6314, interpMode=RCIM_Cubic, leaveTangent=0.6314, tangentMode=RCTM_User, time=1, value=1}
        ]}}}
     CurveIndex=&RandomRangeFloat(Minimum=0, Maximum=1)
     Scale Curve=&MakeCustomFloatFromBool(Bool=$User.bIsRunning, True Float=1300, False Float=800)
    )
   )
   &InheritVelocity(Module.Limit Speed=true, Velocity Scale=<0.1 0.1 0.1>)
  ]
  particleUpdateScript [
   &ScaleColor(
    Scale Alpha=&FloatFromCurve(FloatCurve=DataInterface() {propertyValues {Curve {keys [
         <at 0 1>
         {arriveTangent=-0.0304, interpMode=RCIM_Cubic, leaveTangent=-0.0304, tangentMode=RCTM_User, time=0.6, value=1}
         {arriveTangent=-2.4154, interpMode=RCIM_Cubic, leaveTangent=-2.4154, tangentMode=RCTM_User, time=1, value=0}
        ]}}})
    ScaleA=true
    ScaleRGB=true
   )
   &Drag(Drag=6, Use Linear Drag=true, Use Rotational Drag=false)
   &AccelerationForce(Acceleration=&RandomRangeVector(Minimum=<0 0 -10>, Maximum=<0 0 75>)) {enabled=false}
   &GravityForce(Gravity=&RandomRangeVector(Minimum=<0 0 -75>, Maximum=<0 0 -200>))
   &CurlNoiseForce(
    Noise Frequency=20
    Noise Strength=&FloatFromCurve(
     FloatCurve=DataInterface() {propertyValues {Curve {keys [
         {interpMode=RCIM_Cubic, tangentMode=RCTM_User, time=0, value=1}
         {arriveTangent=-0.0474, interpMode=RCIM_Cubic, leaveTangent=-0.0474, tangentMode=RCTM_User, time=1, value=0.1}
        ]}}}
     Scale Curve=1200
    )
    Pan Noise=false
   )
   &SolveForcesAndVelocity(Clamp Velocity=false, Limit Acceleration=false, Manually Enable Drag Local Offset=true, Manually Enable Rotational Solver=true)
   &ParticleState()
   &UpdateMeshOrientation(Rotation Vector=&Lerp_Vector(
     A=&Multiply_VectorByFloat(Vector=<0.1 0.1 1>, Float=$Particles.RandRotForce)
     B=<0 0 0>
     Alpha=&FloatFromCurve(FloatCurve=DataInterface() {propertyValues {Curve {keys [
          {arriveTangent=1.5391, interpMode=RCIM_Cubic, leaveTangent=1.5391, tangentMode=RCTM_User, time=0, value=0}
          {arriveTangent=1.9127, interpMode=RCIM_Cubic, leaveTangent=1.9127, tangentMode=RCTM_User, time=1, value=1}
         ]}}})
    ))
   &ScaleMeshSize(Scale Factor=&Multiply_VectorByFloat(
     Vector=<1 1 1>
     Float=&FloatFromCurve(FloatCurve=DataInterface() {propertyValues {Curve {keys [
          <at 0 1>
          {arriveTangent=-0.0008, interpMode=RCIM_Cubic, leaveTangent=-0.0008, tangentMode=RCTM_User, time=0.8, value=1}
          {arriveTangent=-3.9673, interpMode=RCIM_Cubic, leaveTangent=-3.9673, tangentMode=RCTM_User, time=1, value=0.2}
         ]}}})
    ))
  ]
  rendererClasses [&NiagaraMeshRendererProperties]
  renderers [{rendererClass=&NiagaraMeshRendererProperties, rendererIndex=0}]
 }
}
summary {
 systemName=NS_Irwin_Nug_FeatherBurst
 emitters [
  {emitterName=zombiefeathers, rendererClasses [&NiagaraMeshRendererProperties]}
  {bEnabled=false, emitterName=recolor_feathers, rendererClasses [&NiagaraMeshRendererProperties]}
 ]
 userVariables [
  User.bIsFrontEnd {type=NiagaraBool}
  User.bIsFrontEndPreview {type=NiagaraBool}
  User.bIsRunning {type=NiagaraBool}
  User.ColorCurve {
   type=NiagaraDataInterfaceColorCurve
   description=""
   defaultValue {
    struct=NiagaraExt_VariableValue_DataInterface
    value {
     dataInterface=&self:NiagaraDataInterfaceColorCurve_0
     dataInterfaceClass=&NiagaraDataInterfaceColorCurve
    }
   }
  }
  User.ColorCurveDataValue {type=NiagaraFloat, defaultValue=0, description=""}
  User.SkeletalMesh {
   type=NiagaraDataInterfaceSkeletalMesh
   defaultValue {
    struct=NiagaraExt_VariableValue_DataInterface
    value {dataInterface=&self:NiagaraDataInterfaceSkeletalMesh_1}
   }
  }
  User.TurnOffFeathersForChrome {type=NiagaraBool}
 ]
}
