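# Physics asset for the Nug bird, a prey AI creature, giving its skeletal mesh a coarse collision shape.
# Four capsule bodies cover the mid neck, the upper spine and the left and right knees, with mirrored capsules on the legs.
# No constraints link the bodies, so the asset defines no jointed ragdoll.
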PhysicsAsset /Irwin/AI/Prey/Nug/Meshes/Nug_Bird_Physics
bodies {
 BipedHeadNeck_A_NeckMid_C [{shapeType=Capsule, length=40, radius=9.25, center=<11.7015 -0.8305 -0.0003>, rotation=<rot 0 -108.96 -89.99>}]
 HoofedLeg_A_Knee_L [{shapeType=Capsule, length=14, radius=10, center=<6.59 4.07 2.28>, rotation=<rot 0 -80.86 -89.68>}]
 HoofedLeg_A_Knee_R [{shapeType=Capsule, length=14, radius=10, center=<-6.59 -4.07 -2.28>, rotation=<rot 0 -80.86 90.32>}]
 QuadSpine_A_Spine3_C [{shapeType=Capsule, length=8.75, radius=20, center=<-6.3098 -0.361 0>, rotation=<rot 90 0 -15.0003>}]
}
constraints []
